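# Animation Blueprint for the CymbalBar backpack cosmetic, posing the pack against the wearer's body.
# Copies the mesh pose, pins bag_spine_04 to the body's spine_04 world transform, caches it, then layers additive pack-bone offsets, one driven by skydive floating.
# The event graph derives alphas from movement states and spine-to-pelvis deltas, including a debug parachute print.
# The dyn_pack_spine bend chains are not connected to the output pose.
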AnimBlueprint /BRCosmetics/Accessories/FORT_Backpacks/Backpack_M_MED_CymbalBar/Meshes/M_MED_CymbalBar_Pack_AnimBP
target_skeleton=&/M_MED_CymbalBar_Pack_Skeleton
graphs [
 AnimGraph {nodes [
   n1 = "Transform (Modify) Bone - Bone: dyn_pack_spine_07"(Alpha=1, Translation="0,0,0", Rotation="10,0,0", Scale="1,1,1") {
    rotationMode=BMM_Additive
    rotationSpace=BCS_BoneSpace
    translationMode=BMM_Additive
    translationSpace=BCS_BoneSpace
    alphaScaleBiasClamp {bInterpResult=true, interpSpeedDecreasing=1, interpSpeedIncreasing=1}
    boneToModify {boneName=dyn_pack_spine_07}
    rotation {pitch=10}
   }
   "Transform (Modify) Bone - Bone: dyn_pack_spine_08"(ComponentPose=@n1, Alpha=1, Translation="0,0,0", Rotation="10,0,0", Scale="1,1,1") {
    rotationMode=BMM_Additive
    rotationSpace=BCS_BoneSpace
    translationMode=BMM_Additive
    translationSpace=BCS_BoneSpace
    alphaScaleBiasClamp {bInterpResult=true, interpSpeedDecreasing=1, interpSpeedIncreasing=1}
    boneToModify {boneName=dyn_pack_spine_08}
    rotation {pitch=10}
   }
   n2 = "Transform (Modify) Bone - Bone: dyn_pack_spine_07"(Alpha=1, Translation="0,0,0", Rotation="15,0,0", Scale="1,1,1") {
    rotationMode=BMM_Additive
    rotationSpace=BCS_BoneSpace
    translationMode=BMM_Additive
    translationSpace=BCS_BoneSpace
    alphaScaleBiasClamp {bInterpResult=true, interpSpeedDecreasing=1, interpSpeedIncreasing=1}
    boneToModify {boneName=dyn_pack_spine_07}
    rotation {pitch=15}
   }
   n3 = "Transform (Modify) Bone - Bone: dyn_pack_spine_08"(ComponentPose=@n2, Alpha=1, Translation="0,0,0", Rotation="10,0,0", Scale="1,1,1") {
    rotationMode=BMM_Additive
    rotationSpace=BCS_BoneSpace
    translationMode=BMM_Additive
    translationSpace=BCS_BoneSpace
    alphaScaleBiasClamp {bInterpResult=true, interpSpeedDecreasing=1, interpSpeedIncreasing=1}
    boneToModify {boneName=dyn_pack_spine_08}
    rotation {pitch=10}
   }
   n4 = "Transform (Modify) Bone - Bone: dyn_pack_spine_03"(ComponentPose=@n3, Alpha=1, Translation="0,0,0", Rotation="25,0,0", Scale="1,1,1") {
    rotationMode=BMM_Additive
    rotationSpace=BCS_BoneSpace
    translationMode=BMM_Additive
    translationSpace=BCS_BoneSpace
    alphaScaleBiasClamp {bInterpResult=true, interpSpeedDecreasing=3, interpSpeedIncreasing=3}
    boneToModify {boneName=dyn_pack_spine_03}
    rotation {pitch=25}
   }
   n5 = "Transform (Modify) Bone - Bone: dyn_pack_spine_03"(ComponentPose=@n4, Alpha=1, Translation="0,0,0", Rotation="30,0,0", Scale="1,1,1") {
    rotationMode=BMM_Additive
    rotationSpace=BCS_BoneSpace
    translationMode=BMM_Additive
    translationSpace=BCS_BoneSpace
    alphaScaleBiasClamp {bInterpResult=true, interpSpeedDecreasing=3, interpSpeedIncreasing=3}
    boneToModify {boneName=dyn_pack_spine_03}
    rotation {pitch=30}
   }
   n6 = "Transform (Modify) Bone - Bone: dyn_pack_spine_04"(ComponentPose=@n5, Alpha=1, Translation="0,0,0", Rotation="30,0,0", Scale="1,1,1") {
    rotationMode=BMM_Additive
    rotationSpace=BCS_BoneSpace
    translationMode=BMM_Additive
    translationSpace=BCS_BoneSpace
    alphaScaleBiasClamp {bInterpResult=true, interpSpeedDecreasing=3, interpSpeedIncreasing=3}
    boneToModify {boneName=dyn_pack_spine_04}
    rotation {pitch=30}
   }
   "Transform (Modify) Bone - Bone: dyn_pack_spine_06"(ComponentPose=@n6, Alpha=1, Translation="0,0,0", Rotation="20,0,0", Scale="1,1,1") {
    rotationMode=BMM_Additive
    rotationSpace=BCS_BoneSpace
    translationMode=BMM_Additive
    translationSpace=BCS_BoneSpace
    alphaScaleBiasClamp {bInterpResult=true, interpSpeedDecreasing=3, interpSpeedIncreasing=3}
    boneToModify {boneName=dyn_pack_spine_06}
    rotation {pitch=20}
   }
   n7 = Get inJumpingAir() -> (inJumpingAir:double)
   n8 = Get SpineTwistAlpha() -> (SpineTwistAlpha:float)
   n9 = "Transform (Modify) Bone - Bone: dyn_pack_spine_04"(
    ComponentPose=Reroute Node(InputPin:Component Space Pose Link Structure) -> (OutputPin:Component Space Pose Link Structure)
    Alpha=@n8
    Translation="0,0,1"
    Rotation="5,0,0"
    Scale="1,1,1"
   ) {
    rotationMode=BMM_Additive
    scaleSpace=BCS_WorldSpace
    translationMode=BMM_Additive
    boneToModify {boneName=dyn_pack_spine_04}
    rotation {pitch=5}
    translation {z=1}
   }
   n10 = "Transform (Modify) Bone - Bone: dyn_pack_spine_05"(ComponentPose=@n9, Alpha=@n8, Translation="0,0,1", Rotation="5,0,0", Scale="1,1,1") {
    rotationMode=BMM_Additive
    rotationSpace=BCS_BoneSpace
    scaleSpace=BCS_WorldSpace
    translationMode=BMM_Additive
    translationSpace=BCS_BoneSpace
    boneToModify {boneName=dyn_pack_spine_05}
    rotation {pitch=5}
    translation {z=1}
   }
   n11 = "Transform (Modify) Bone - Bone: dyn_pack_spine_06"(ComponentPose=@n10, Alpha=@n8, Translation="0,0,1", Rotation="5,0,0", Scale="1,1,1") {
    rotationMode=BMM_Additive
    rotationSpace=BCS_BoneSpace
    scaleSpace=BCS_WorldSpace
    translationMode=BMM_Additive
    translationSpace=BCS_BoneSpace
    boneToModify {boneName=dyn_pack_spine_06}
    rotation {pitch=5}
    translation {z=1}
   }
   n12 = "Transform (Modify) Bone - Bone: dyn_pack_spine_07"(ComponentPose=@n11, Alpha=@n8, Translation="0,0,1", Rotation="5,0,0", Scale="1,1,1") {
    rotationMode=BMM_Additive
    rotationSpace=BCS_BoneSpace
    scaleSpace=BCS_WorldSpace
    translationMode=BMM_Additive
    translationSpace=BCS_BoneSpace
    boneToModify {boneName=dyn_pack_spine_07}
    rotation {pitch=5}
    translation {z=1}
   }
   n13 = "Transform (Modify) Bone - Bone: dyn_pack_spine_08"(ComponentPose=@n12, Alpha=@n8, Translation="0,-1,0", Rotation="5,0,0", Scale="1,1,1") {
    rotationMode=BMM_Additive
    rotationSpace=BCS_BoneSpace
    scaleSpace=BCS_WorldSpace
    translationMode=BMM_Additive
    translationSpace=BCS_BoneSpace
    boneToModify {boneName=dyn_pack_spine_08}
    rotation {pitch=5}
    translation {y=-1}
   }
   n14 = "Transform (Modify) Bone - Bone: dyn_pack_spine_04"(ComponentPose=@n13, Alpha=@n7, Translation="0,0,0", Rotation="5,0,0", Scale="1,1,1") {
    rotationMode=BMM_Additive
    rotationSpace=BCS_BoneSpace
    scaleSpace=BCS_WorldSpace
    translationMode=BMM_Additive
    translationSpace=BCS_BoneSpace
    alphaScaleBiasClamp {bInterpResult=true, interpSpeedDecreasing=2, interpSpeedIncreasing=2}
    boneToModify {boneName=dyn_pack_spine_04}
    rotation {pitch=5}
   }
   n15 = "Transform (Modify) Bone - Bone: dyn_pack_spine_05"(ComponentPose=@n14, Alpha=@n7, Translation="0,0,0", Rotation="5,0,0", Scale="1,1,1") {
    rotationMode=BMM_Additive
    rotationSpace=BCS_BoneSpace
    scaleSpace=BCS_WorldSpace
    translationMode=BMM_Additive
    translationSpace=BCS_BoneSpace
    alphaScaleBiasClamp {bInterpResult=true, interpSpeedDecreasing=2, interpSpeedIncreasing=2}
    boneToModify {boneName=dyn_pack_spine_05}
    rotation {pitch=5}
   }
   n16 = "Transform (Modify) Bone - Bone: dyn_pack_spine_06"(ComponentPose=@n15, Alpha=@n7, Translation="0,0,0", Rotation="5,0,0", Scale="1,1,1") {
    rotationMode=BMM_Additive
    rotationSpace=BCS_BoneSpace
    scaleSpace=BCS_WorldSpace
    translationMode=BMM_Additive
    translationSpace=BCS_BoneSpace
    alphaScaleBiasClamp {bInterpResult=true, interpSpeedDecreasing=2, interpSpeedIncreasing=2}
    boneToModify {boneName=dyn_pack_spine_06}
    rotation {pitch=5}
   }
   n17 = "Transform (Modify) Bone - Bone: dyn_pack_spine_07"(ComponentPose=@n16, Alpha=@n7, Translation="0,0,0", Rotation="10,0,0", Scale="1,1,1") {
    rotationMode=BMM_Additive
    rotationSpace=BCS_BoneSpace
    scaleSpace=BCS_WorldSpace
    translationMode=BMM_Additive
    translationSpace=BCS_BoneSpace
    alphaScaleBiasClamp {bInterpResult=true, interpSpeedDecreasing=2, interpSpeedIncreasing=2}
    boneToModify {boneName=dyn_pack_spine_07}
    rotation {pitch=10}
   }
   n18 = "Transform (Modify) Bone - Bone: dyn_pack_spine_08"(ComponentPose=@n17, Alpha=@n7, Translation="0,0,0", Rotation="10,0,0", Scale="1,1,1") {
    rotationMode=BMM_Additive
    rotationSpace=BCS_BoneSpace
    scaleSpace=BCS_WorldSpace
    translationMode=BMM_Additive
    translationSpace=BCS_BoneSpace
    alphaScaleBiasClamp {bInterpResult=true, interpSpeedDecreasing=2, interpSpeedIncreasing=2}
    boneToModify {boneName=dyn_pack_spine_08}
    rotation {pitch=10}
   }
   n19 = "Transform (Modify) Bone - Bone: dyn_pack_spine_04"!(ComponentPose=@n18, bAlphaBoolEnabled:bool=True, Translation="0,0,0", Rotation="10,0,0.01", Scale="1,1,1") -> (Pose) {
    alphaInputType=Bool
    rotationMode=BMM_Additive
    rotationSpace=BCS_BoneSpace
    scaleSpace=BCS_WorldSpace
    translationMode=BMM_Additive
    translationSpace=BCS_BoneSpace
    alphaScaleBiasClamp {bInterpResult=true, interpSpeedDecreasing=2, interpSpeedIncreasing=2}
    boneToModify {boneName=dyn_pack_spine_04}
    rotation {pitch=10, roll=0.01}
   }
   n20 = "Transform (Modify) Bone - Bone: dyn_pack_spine_06"!(ComponentPose=@n19, bAlphaBoolEnabled:bool=True, Translation="0,0,0", Rotation="10,0,0.01", Scale="1,1,1") -> (Pose) {
    alphaInputType=Bool
    rotationMode=BMM_Additive
    rotationSpace=BCS_BoneSpace
    scaleSpace=BCS_WorldSpace
    translationMode=BMM_Additive
    translationSpace=BCS_BoneSpace
    alphaScaleBiasClamp {bInterpResult=true, interpSpeedDecreasing=2, interpSpeedIncreasing=2}
    boneToModify {boneName=dyn_pack_spine_06}
    rotation {pitch=10, roll=0.01}
   }
   "Transform (Modify) Bone - Bone: dyn_pack_spine_08"!(ComponentPose=@n20, bAlphaBoolEnabled:bool=True, Translation="0,0,0", Rotation="8,0,0.01", Scale="1,1,1") -> (Pose) {
    alphaInputType=Bool
    rotationMode=BMM_Additive
    rotationSpace=BCS_BoneSpace
    scaleSpace=BCS_WorldSpace
    translationMode=BMM_Additive
    translationSpace=BCS_BoneSpace
    alphaScaleBiasClamp {bInterpResult=true, interpSpeedDecreasing=2, interpSpeedIncreasing=2}
    boneToModify {boneName=dyn_pack_spine_08}
    rotation {pitch=8, roll=0.01}
   }
   n21 = "Transform (Modify) Bone - Bone: pack"(
    ComponentPose=Local To Component(LocalPose=Use cached pose 'CR_Cache_Input'())
    Alpha=Get SkyDiveFloating_Alpha() -> (SkyDiveFloating_Alpha:double)
    Translation="0,0,1"
    Rotation="0,0,20"
   ) {
    rotationMode=BMM_Additive
    rotationSpace=BCS_BoneSpace
    translationMode=BMM_Additive
    translationSpace=BCS_BoneSpace
    alphaScaleBiasClamp {bInterpResult=true, interpSpeedDecreasing=1, interpSpeedIncreasing=1}
    boneToModify {boneName=pack}
    rotation {roll=20}
    translation {z=1}
   }
   n22 = "Transform (Modify) Bone - Bone: pack"(ComponentPose=@n21, Translation="0,0,0", Rotation="-0.000,0.000,35.820") {
    rotationMode=BMM_Additive
    rotationSpace=BCS_BoneSpace
    translationMode=BMM_Additive
    translationSpace=BCS_BoneSpace
    alphaScaleBiasClamp {bInterpResult=true, interpSpeedDecreasing=1, interpSpeedIncreasing=1}
    boneToModify {boneName=pack}
    rotation {roll=35.82}
   }
   n23 = "Transform (Modify) Bone - Bone: pack"(ComponentPose=@n22, Translation="-0.000,-2.109,0.372", Rotation="-0.000,0.000,11.499") {
    rotationMode=BMM_Additive
    rotationSpace=BCS_BoneSpace
    translationMode=BMM_Additive
    translationSpace=BCS_BoneSpace
    alphaScaleBiasClamp {bInterpResult=true}
    boneToModify {boneName=pack}
    rotation {roll=11.499}
    translation {y=-2.109, z=0.372}
   }
   n24 = "Transform (Modify) Bone - Bone: pack"(ComponentPose=@n23, Translation="0,0,0", Rotation="0,0,10") {
    rotationMode=BMM_Additive
    rotationSpace=BCS_BoneSpace
    translationMode=BMM_Additive
    translationSpace=BCS_BoneSpace
    alphaScaleBiasClamp {bInterpResult=true}
    boneToModify {boneName=pack}
    rotation {roll=10}
   }
   n25 = "Transform (Modify) Bone - Bone: pack"(ComponentPose=@n24, Translation="0,-15,0", Rotation="0,0,30") {
    rotationMode=BMM_Additive
    rotationSpace=BCS_BoneSpace
    translationMode=BMM_Additive
    translationSpace=BCS_BoneSpace
    alphaScaleBiasClamp {bInterpResult=true}
    boneToModify {boneName=pack}
    rotation {roll=30}
    translation {y=-15}
   }
   Output Pose(Result=Component To Local(ComponentPose=@n25)) {layerGroup=None}
   n26 = "Transform (Modify) Bone - Bone: body_root"(ComponentPose=Local To Component(LocalPose=Copy Pose From Mesh()), Alpha=1, Translation="0,0,0", Rotation="0,0,0", Scale="1,1,1") {
    rotationMode=BMM_Replace
    rotationSpace=BCS_WorldSpace
    scaleSpace=BCS_BoneSpace
    translationMode=BMM_Replace
    translationSpace=BCS_WorldSpace
    alphaBoolBlend {blendOption=Linear}
    alphaScaleBiasClamp {bInterpResult=true, interpSpeedDecreasing=1, interpSpeedIncreasing=1, inRange {max=0, min=-40}, outRange {max=0, min=1}}
    boneToModify {boneName=body_root}
   }
   n27 = "Transform (Modify) Bone - Bone: bag_pelvis"(ComponentPose=@n26, Alpha=1, Translation="0,0,0", Rotation="0,0,0", Scale="1,1,1") {
    rotationMode=BMM_Replace
    rotationSpace=BCS_WorldSpace
    scaleSpace=BCS_BoneSpace
    translationMode=BMM_Replace
    translationSpace=BCS_WorldSpace
    alphaBoolBlend {blendOption=Linear}
    alphaScaleBiasClamp {inRange {max=0, min=-40}, outRange {max=0, min=1}}
    boneToModify {boneName=bag_pelvis}
   }
   n28 = "Transform (Modify) Bone - Bone: bag_spine_04"(
    ComponentPose=@n27
    Alpha=1
    Translation=Get spine_04_location() -> (spine_04_location:Vector)
    Rotation=Get spine_04_rotation() -> (spine_04_rotation:Rotator)
    Scale="1,1,1"
   ) {
    rotationMode=BMM_Replace
    rotationSpace=BCS_WorldSpace
    scaleSpace=BCS_BoneSpace
    translationMode=BMM_Replace
    translationSpace=BCS_WorldSpace
    alphaBoolBlend {blendOption=Linear}
    alphaScaleBiasClamp {inRange {max=0, min=-40}, outRange {max=0, min=1}}
    boneToModify {boneName=bag_spine_04}
   }
   Save cached pose 'CR_Cache_Input'(Pose=Component To Local(ComponentPose=@n28)) {cacheName=CR_Cache_Input}
   n29 = "Transform (Modify) Bone - Bone: dyn_pack_spine_05"(Alpha=1, Translation="0,0,0", Rotation="15,0,0", Scale="1,1,1") {
    rotationMode=BMM_Additive
    rotationSpace=BCS_BoneSpace
    translationMode=BMM_Additive
    translationSpace=BCS_BoneSpace
    alphaScaleBiasClamp {bInterpResult=true, interpSpeedDecreasing=1, interpSpeedIncreasing=1}
    boneToModify {boneName=dyn_pack_spine_05}
    rotation {pitch=15}
   }
   n30 = "Transform (Modify) Bone - Bone: dyn_pack_spine_06"!(ComponentPose=@n29, Alpha:float=1, Translation="0,0,0", Rotation="10,0,0", Scale="1,1,1") -> (Pose) {
    rotationMode=BMM_Additive
    rotationSpace=BCS_BoneSpace
    translationMode=BMM_Additive
    translationSpace=BCS_BoneSpace
    alphaScaleBiasClamp {bInterpResult=true, interpSpeedDecreasing=1, interpSpeedIncreasing=1}
    boneToModify {boneName=dyn_pack_spine_06}
    rotation {pitch=10}
   }
   Reroute Node(InputPin=@n30)
  ]}
 EventGraph {nodes [
   Event OnBeginPlayingEmote() {is=title K2Node_Event}
   > Set EmoteOn(EmoteOn:double=1) -> (Output_Get:double)
   Event OnBeginPickaxeSwing() {is=title K2Node_Event}
   > Set pickaxeSwing(pickaxeSwing:double=1) -> (Output_Get:double)
   Event OnParachuteClosed()
   > Set ParachuteBool(ParachuteBool:bool=false) -> (Output_Get:bool)
   Event OnParachuteOpened()
   > Set ParachuteBool(ParachuteBool=true)
   Event OnEndPickaxeSwing() {is=title K2Node_Event}
   > Set pickaxeSwing(pickaxeSwing=0)
   Event OnEndPlayingEmote() {is=title K2Node_Event}
   > Set EmoteOn(EmoteOn=0)
   Event BlueprintUpdateAnimation()
   > n1 = Sequence!(execute) -> (then_0, then_1, then_2:Exec, then_3:Exec, then_4:Exec, then_5:Exec, then_6:Exec)
   n2 = Get bIsSkydiveDiveMode()
   n3 = Get bIsParachuteOpened()
   "Set SkyDiveMode&&Parachute_Alpha"(execute=@n1.then_1, "SkyDiveMode&&Parachute_Alpha":double="To Float (Boolean)"(InBool=OR Boolean(A=@n2, B=@n3))) -> (Output_Get:double)
   > Set SkyDiveFloating_Alpha(SkyDiveFloating_Alpha:double="To Float (Boolean)"(InBool=Get bIsSkydiveFloating())) -> (Output_Get:double)
   > Set PickaxeSwingBool(PickaxeSwingBool:bool="Equal (Float)"(A=Get pickaxeSwing() -> (pickaxeSwing:double), B=1)) -> (Output_Get:bool)
   n4 = Branch(execute=@n1.then_2, Condition=Get PickaxeSwingBool() -> (PickaxeSwingBool:bool))
   > Set SpineTwistAlpha(SpineTwistAlpha:float=GetAlphaFromReferenceTransformDelta(
     SkeletalMeshComponent=Get Skeletal Mesh() -> (Skeletal Mesh:Skeletal Mesh Component Object Reference)
     BoneName=bag_spine_04
     BaseName=bag_pelvis
     DeltaComponentType=EulerX
     InputRange=(X=-5,Y=-25)
     TwistAxis="1,0,0"
    )) -> (Output_Get:float)
   Set SpineTwistAlpha(execute=@n4.else, SpineTwistAlpha=0)
   n5 = Branch(execute=@n1.then_3, Condition=Get ParachuteBool() -> (ParachuteBool:bool))
   > Set SpineGlidingAlpha(SpineGlidingAlpha:float=GetAlphaFromReferenceTransformDelta(SkeletalMeshComponent=Get Skeletal Mesh(), BoneName=bag_spine_04, BaseName=bag_pelvis, InputRange=(X=-20,Y=-30), TwistAxis="1,0,0")) -> (Output_Get:float)
   Set SpineGlidingAlpha(execute=@n5.else, SpineGlidingAlpha=0)
   n6 = NOT Boolean(A=Get bIsSkyDiving())
   n7 = Get bIsInAir() -> (bIsInAir:bool)
   Set inJumpingAir(execute=@n1.then_4, inJumpingAir:double="To Float (Boolean)"(InBool=AND Boolean(A=@n7, B=@n6))) -> (Output_Get:double)
   n8 = Contains(SearchIn=GetDisplayName(Object=Get Body_Mesh() -> (Body_Mesh:Scene Component Object Reference)), Substring=M_, bUseCase=true)
   n9 = NOT Boolean(A=Get bIsTacticalSprinting())
   n10 = Get bIsMoving2D()
   > Set isRunningCustom(isRunningCustom:double="To Float (Boolean)"(InBool=AND Boolean!(A=@n10, B=@n8, C:bool=@n9) -> (ReturnValue))) -> (Output_Get:double)
   n11 = Get bIsMoving2D()
   n12 = Get bIsTacticalSprinting()
   > Set isTacticalRunningCustom(isTacticalRunningCustom:double="To Float (Boolean)"(InBool=AND Boolean(A=@n11, B=@n12, C=@n8))) -> (Output_Get:double)
   Set Skeletal Mesh(execute=@n1.then_0, Skeletal Mesh:Skeletal Mesh Component Object Reference=GetPartSkeletalMeshComponent()) -> (Output_Get:Skeletal Mesh Component Object Reference)
   n13 = GetPartSkeletalMeshComponent()
   > Is Valid(InputObject=@n13)
   n14 = GetAttachParent!(self:Scene Component Object Reference=@n13) -> (ReturnValue:Scene Component Object Reference)
   > Is Valid(InputObject=@n14)
   > Set Body_Mesh(Body_Mesh:Scene Component Object Reference=@n14) -> (Output_Get:Scene Component Object Reference)
   > n15 = Sequence!(execute) -> (then_0, then_1, then_2, then_3, then_4, then_5, then_6, then_7:Exec)
   n16 = Get Body_Mesh()
   n17 = BreakTransform(InTransform=GetSocketTransform(self=@n16, InSocketName=root, TransformSpace=RTS_World))
   Set root_location(execute=@n15.then_6, root_location:Vector=@n17) -> (Output_Get:Vector)
   > Set root_space_rotation(root_space_rotation:Rotator=@n17.Rotation) -> (Output_Get:Rotator)
   n18 = BreakTransform(InTransform=GetSocketTransform(self=@n16, InSocketName=pelvis, TransformSpace=RTS_World))
   Set pelvis_location(execute=@n15.then_0, pelvis_location:Vector=@n18) -> (Output_Get:Vector)
   > Set pelvis_rotation(pelvis_rotation:Rotator=@n18.Rotation) -> (Output_Get:Rotator)
   n19 = BreakTransform(InTransform=GetSocketTransform(self=@n16, InSocketName=root, TransformSpace=RTS_World))
   Set body_root_location(execute=@n15.then_5, body_root_location:Vector=@n19) -> (Output_Get:Vector)
   > Set body_root_rotation(body_root_rotation:Rotator=@n19.Rotation) -> (Output_Get:Rotator)
   n20 = BreakTransform(InTransform=GetSocketTransform(self=@n16, InSocketName=spine_04, TransformSpace=RTS_World))
   Set spine_04_location(execute=@n15.then_7, spine_04_location:Vector=@n20) -> (Output_Get:Vector)
   > Set spine_04_rotation(spine_04_rotation:Rotator=@n20.Rotation) -> (Output_Get:Rotator)
   Set isCrouchingFloat(execute=@n1.then_5, isCrouchingFloat:double="To Float (Boolean)"(InBool=Get bIsCrouching())) -> (Output_Get:double)
   > Set SpineDelta_Z_only(SpineDelta_Z_only:float=GetAlphaFromReferenceTransformDelta(SkeletalMeshComponent=Get Skeletal Mesh(), BoneName=bag_spine_04, BaseName=bag_pelvis, InputRange=(X=-25,Y=-50), TwistAxis="1,0,0")) -> (Output_Get:float)
   > n21 = Branch(Condition=Get bIsParachuteOpened())
   > n22 = Set Parachute_Delta_Rot(Parachute_Delta_Rot:float=Get SpineDelta_Z_only() -> (SpineDelta_Z_only:float)) -> (Output_Get:float)
   > PrintString!(execute, InString="To String (Float)"(InDouble=@n22), bPrintToScreen, bPrintToLog, TextColor, Duration, Key) -> (then)
   Set Parachute_Delta_Rot(execute=@n21.else, Parachute_Delta_Rot=0)
   n23 = GetAlphaFromReferenceTransformDelta(SkeletalMeshComponent=Get Skeletal Mesh(), BoneName=bag_spine_04, BaseName=bag_pelvis, InputRange=(X=10,Y=35), TwistAxis="1,0,0")
   Set SpineDelta_only(execute=@n1.then_6, SpineDelta_only:float=@n23) -> (Output_Get:float)
   > n24 = Branch(Condition="Equal (Float)"(A=Get EmoteOn() -> (EmoteOn:double), B=1))
   > Set Spine_Twist_EmoteOn(Spine_Twist_EmoteOn:double=@n23) -> (Output_Get:double)
   Set Spine_Twist_EmoteOn(execute=@n24.else, Spine_Twist_EmoteOn=0)
   TryGetPawnOwner()
   TryGetPawnOwner()
   Get bIsInAir()
   Get ParachuteBool()
  ]}
]
variables [
 Body_Mesh {type=SceneComponent, default=null}
 body_root_location {type=Vector}
 body_root_rotation {type=Rotator}
 EmoteOn {type=number}
 inJumpingAir {type=number}
 isCrouchingFloat {type=number}
 isRunningCustom {type=number}
 isTacticalRunningCustom {type=number}
 Parachute_Delta_Rot {type=number}
 ParachuteBool {type=boolean}
 pelvis_location {type=Vector}
 pelvis_rotation {type=Rotator}
 pickaxeSwing {type=number}
 PickaxeSwingBool {type=boolean}
 root_location {type=Vector}
 root_space_rotation {type=Rotator}
 Skeletal Mesh {type=SkeletalMeshComponent}
 Skydive {type=number}
 SkyDiveFloating_Alpha {type=number}
 "SkyDiveMode&&Parachute_Alpha" {type=number}
 spine_04_location {type=Vector}
 spine_04_rotation {type=Rotator}
 Spine_Twist_EmoteOn {type=number}
 SpineDelta_only {type=number}
 SpineDelta_Z_only {type=number}
 SpineGlidingAlpha {type=number}
 SpineTwistAlpha {type=number}
]
